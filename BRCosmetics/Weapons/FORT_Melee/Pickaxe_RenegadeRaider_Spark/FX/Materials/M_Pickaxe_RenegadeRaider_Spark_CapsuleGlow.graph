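# Additive unlit Niagara sprite material for a capsule-shaped glow on the Renegade Raider Spark pickaxe.
# Sprite_Capsule stretches the sprite using dynamic-parameter direction and scale, though only Direction X fills all three direction components.
# Its offset adds a push toward the camera.
# A soft sphere mask on capsule UVs, sharpened by exponent, drives depth-faded opacity with particle alpha; particle color and brightness feed emissive adjustments.
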Material /BRCosmetics/Weapons/FORT_Melee/Pickaxe_RenegadeRaider_Spark/FX/Materials/M_Pickaxe_RenegadeRaider_Spark_CapsuleGlow
graph {
 properties {bUsedWithNiagaraSprites=true, blendMode=BLEND_Additive, pixelDepthOffsetMode=PDOM_Legacy, shadingModel=MSM_Unlit, usedShadingModels=Unlit}
 nodes [
  n1 = DynamicParameter() {paramNames [Particle Direction X, Particle Direction Y, Particle Direction Z, Scale Factor]}
  n2 = MaterialFunctionCall!(
   Particle Position
   Normalized Particle Direction=MaterialFunctionCall(R=@n1.Particle Direction X, G=@n1.Particle Direction X, B=@n1.Particle Direction X) {materialFunction=&AppendMany}.RGB
   Capsule Diameter
   Capsule Pivot Alignment
   Scale Along Velocity Vector=@n1.Scale Factor
   ---------Scale Over Velocity Parameters---------
   Particle Speed
   Minimum Speed
   Max Speed
   Minimum Scale Factor
   Maximum Scale Factor
   ---------- Advanced -----------
   Cap Size Multiplier
   Use Camera Offset
   Use Advanced Features
   World Position to Negate
  ) {materialFunction=/Engine/Functions/Engine_MaterialFunctions02/WorldPositionOffset/Sprite_Capsule}
  n3 = Subtract(A=CameraPositionWS(), B=Add(A=@n2.World Position Offset, B=WorldPosition().XYZ))
  Add(A=@n2.World Position Offset, B=Multiply(A=Normalize(VectorInput=@n3), B=ScalarParameter() {defaultValue=33, parameterName=Camera Offset}))
  n4 = ParticleColor()
  n5 = Constant() {r=0.5}
  n6 = SphereMask(A=@n2.Texture Coordinates, B=@n5, Radius=@n5) {hardnessPercent=0}
  n7 = ScalarParameter() {defaultValue=3, parameterName=Opacity Exponent}
  DepthFade(Opacity=Multiply(A=@n4.A, B=Power(Base=@n6, Exp=@n7)), FadeDistance=ScalarParameter() {defaultValue=66, parameterName=Depth Fade Distance})
  MaterialFunctionCall(Emissive=Multiply(A=@n4.RGB, B=ScalarParameter() {defaultValue=0.1, parameterName=Brightness})) {materialFunction=&MF_FX_EmissiveAdjustments}
 ]
}
